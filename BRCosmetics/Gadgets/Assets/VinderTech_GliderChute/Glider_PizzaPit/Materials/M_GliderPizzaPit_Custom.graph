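# Subsurface material for the Pizza Pit glider, adding animated dripping cheese to the skeletal mesh.
# A cheese mask built from pre-skinned distance to a center and upward direction drives tint, fresnel sheen, specular, roughness and subsurface glow.
# Within it, a waterfall normal, world-aligned-projected on panning pre-skinned positions, blends over the base normal.
# Sine-based vertex offsets add cheese wobble and broader wind wobble before attributes feed the accessory tech-art override function.
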Material /BRCosmetics/Gadgets/Assets/VinderTech_GliderChute/Glider_PizzaPit/Materials/M_GliderPizzaPit_Custom
graph {
 properties {
  bUseMaterialAttributes=true
  bUsedWithMeshDeformer=true
  bUsedWithSkeletalMesh=true
  pixelDepthOffsetMode=PDOM_Legacy
  shadingModel=MSM_Subsurface
  usedShadingModels=Subsurface
 }
 nodes [
  n1 = TextureSample() {samplerType=SAMPLERTYPE_Normal, texture=/Game/Gadgets/Assets/VinderTech_GliderChute/Glider_PizzaPit/Textures/T_Glider_PizzaPit_N}
  n2 = ScalarParameter() {defaultValue=602.9581, parameterName=Normalsize}
  n3 = AppendVector(A=@n2, B=@n2)
  n4 = Multiply(A=Multiply(A=Time(), B=Constant3Vector() {constant=<rgba 0 0 1 0>}), B=ScalarParameter() {defaultValue=4.1016, parameterName=PanSpeed})
  n5 = PreSkinnedPosition()
  n6 = Add(A=@n4, B=@n5)
  n7 = MaterialFunctionCall(
   Normal=MaterialFunctionCall(
    TextureObject=TextureObject() {samplerType=SAMPLERTYPE_Normal, texture=/Game/Packages/Fortnite_UniqueMaterials_1/Textures/T_WaterFall_Tile_N}
    TextureSize=AppendVector(A=@n3, B=@n2)
    ProjectionTransitionContrast=ScalarParameter() {defaultValue=-1.2047, parameterName=Contrast}
    WorldPosition=VertexInterpolator(VS=@n6).PS
   ) {materialFunction=&WorldAlignedNormal}.XYZFlatTop
   Flatness=ScalarParameter() {defaultValue=-0.6651, parameterName=DrippyNormalFlatten}
  ) {materialFunction=&FlattenNormal}
  Add(A=@n1.RGB, B=Multiply(A=@n7.Result, B=Constant3Vector() {constant=<rgba 1 1 0 0>}))
  Constant3Vector() {constant=<rgba 0.9219 0.0528 0 0>}
  Frac(None=@n4)
  n8 = PreSkinnedPosition()
  n9 = VectorParameter() {parameterName=Center, defaultValue=<rgba 0 0 0 1>}
  n10 = Add(A=Distance(A=@n8, B=@n9.RGB), B=ScalarParameter() {defaultValue=-40.304, parameterName=DistanceBlendOffset})
  n11 = ScalarParameter() {defaultValue=6.5483, parameterName=DistanceBlend}
  n12 = Subtract(A=@n8, B=@n9.RGB)
  n13 = Clamp(None=DotProduct(A=Normalize(VectorInput=@n12), B=Constant3Vector() {constant=<rgba 0 0 1 0>}))
  n14 = Clamp(None=Multiply(A=Divide(A=@n10, B=@n11), B=@n13))
  n15 = ComponentMask(None=VertexInterpolator(VS=@n14).PS) {r=true}
  n16 = Multiply(
   A=LinearInterpolate(
    A=Constant3Vector() {constant=<rgba 1 1 1 0>}
    B=VectorParameter() {parameterName=CheeseColor, defaultValue=<rgba 0.5312 0.5312 0.5312 1>}.RGB
    Alpha=@n15
   )
   B=TextureSample() {texture=/Game/Gadgets/Assets/VinderTech_GliderChute/Glider_PizzaPit/Textures/T_Glider_PizzaPit_D}.RGB
  )
  n17 = VectorParameter() {parameterName=SSSColor, defaultValue=<rgba 1 0.5807 0.2077 1>}
  n18 = Multiply(A=ScalarParameter() {defaultValue=1.1663, parameterName=SSSBrightness}, B=@n15)
  n19 = TextureSample() {samplerType=SAMPLERTYPE_Masks, texture=/Game/Gadgets/Assets/VinderTech_GliderChute/Glider_PizzaPit/Textures/T_Glider_PizzaPit_S}
  n20 = LinearInterpolate(
   A=ScalarParameter() {defaultValue=-1.3356, parameterName=RoughnessCheese}
   B=ScalarParameter() {defaultValue=3.4259, parameterName=RoughnessCheese2}
   Alpha=@n19.B
  )
  n21 = ScalarParameter() {defaultValue=4.8913, parameterName=FresnellPower}
  n22 = ScalarParameter() {defaultValue=2.3128, parameterName=FresnellPowerhigh}
  n23 = ScalarParameter() {defaultValue=3.2394, parameterName=FresnellBrightness1}
  n24 = ScalarParameter() {defaultValue=-3.1301, parameterName=FresnellBrightness2}
  n25 = Multiply(
   A=Fresnel(ExponentIn=LinearInterpolate(A=@n21, B=@n22, Alpha=@n19.B), Normal=Transform(None=@n1.RGB)) {baseReflectFraction=0}
   B=Clamp(None=LinearInterpolate(A=@n23, B=@n24, Alpha=@n19.B))
  )
  n26 = Time()
  n27 = ScalarParameter() {defaultValue=95.1972, parameterName=WobbleWPOSpeed}
  n28 = Divide(A=Add(A=@n5, B=Multiply(A=@n26, B=@n27)), B=ScalarParameter() {defaultValue=254.3462, parameterName=WobbleWPOWaveLength})
  n29 = Multiply(
   A=Multiply(A=VertexNormalWS(), B=ScalarParameter() {defaultValue=1, parameterName=WobbleWPODistance})
   B=Add(A=Sine(None=@n28), B=Constant() {r=0.5})
  )
  n30 = ComponentMask(None=@n14) {r=true}
  n31 = ScalarParameter() {defaultValue=198.527, parameterName=DistanceBlend2}
  n32 = Clamp(None=Multiply(A=Divide(A=@n10, B=@n31), B=@n13))
  n33 = Subtract(A=@n5, B=Constant3Vector())
  n34 = Time()
  n35 = ScalarParameter() {defaultValue=-186.9569, parameterName=WobbleWPOSpeed2}
  n36 = Divide(
   A=Add(A=ComponentMask(None=@n5) {g=true}, B=Multiply(A=@n34, B=@n35))
   B=ScalarParameter() {defaultValue=84.995, parameterName=WobbleWPOWaveLength2}
  )
  n37 = Multiply(A=Multiply(A=Normalize(VectorInput=@n33), B=Sine(None=@n36)), B=ScalarParameter() {defaultValue=1, parameterName=WIndWobble})
  n38 = SetMaterialAttributes!(
   MaterialAttributes
   Base Color=Add(A=Multiply(A=@n25, B=@n15), B=@n16)
   Metallic=@n19.G
   Specular=LinearInterpolate(A=@n19.R, B=ScalarParameter() {defaultValue=1, parameterName=SpecularCheese}, Alpha=@n15)
   Roughness=LinearInterpolate(A=@n19.B, B=Multiply(A=@n20, B=@n19.B), Alpha=@n15)
   Opacity=ScalarParameter() {defaultValue=0.0857, parameterName=SSSOpcaity}
   Normal=LinearInterpolate(
    A=@n1.RGB
    B=MaterialFunctionCall(BaseNormal=@n1.RGB, AdditionalNormal=@n7.Result) {materialFunction=&BlendAngleCorrectedNormals}.Result
    Alpha=@n15
   )
   World Position Offset=Add(A=Multiply(A=@n29, B=@n30), B=Multiply(A=@n32, B=@n37))
   Subsurface Color=Multiply(A=@n16, B=Multiply(A=@n17.RGB, B=@n18))
  ) {attributeSetTypes [
    #1
    #2
    #3
    #4
    #5
    #6
    #7
    #8
   ]}
  MaterialFunctionCall(In=@n38) {materialFunction=&MF_TechArtOverrides_accessories}
  Multiply(B=Transform(None=Constant3Vector() {constant=<rgba 0 0 1 0>}) {transformSourceType=TRANSFORMSOURCE_Local})
  TextureSample() {samplerType=SAMPLERTYPE_Normal, texture=&T_WaterFall_Tile_N}
  TransformPosition(None=@n5) {transformType=TRANSFORMPOSSOURCE_World}
  VertexInterpolator()
  WorldPosition()
 ]
}
